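# Physics asset for the Dragon Ninja pickaxe, a two-handed melee weapon cosmetic in Fortnite Battle Royale.
# It holds a single body on the Melee_TwoHanded_Handle bone, shaped as one large capsule offset and rotated along the weapon.
# There are no constraints, so the mesh collides as one rigid piece with no jointed or secondary physics parts.
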PhysicsAsset /BRCosmetics/Weapons/FORT_Melee/Pickaxe_Dragon_Ninja/Mesh/Pickaxe_Dragon_Ninja_PhysicsAsset
bodies {Melee_TwoHanded_Handle [{shapeType=Capsule, length=98.1939, radius=40.6858, center=<-0.0257 7.4753 24.7789>, rotation=<rot 1.701 -90 0.0152>}]}
constraints []
